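; StorePinataMaster_Parent is a Blueprint parent class for the pinata in the front-end store.
; It declares three custom events: InitiatePinata, CameraResetComplete and CameraTransitionComplete, covering pinata start and camera completion callbacks.
; All three events and the construction script are empty, so the class acts as an interface-like base with no behavior of its own.
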(event Custom|InitiatePinata)

(event Custom|CameraResetComplete)

(event Custom|CameraTransitionComplete)

(fn ConstructionScript)
